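# First-person firing sound cue for an RPG launcher.
# An equally weighted random node picks one of three close-range fire waves, then a modulator holds volume constant and leaves pitch variation at its default.
# The boosted cue uses first-person projectile attenuation, the first-person ranged-fire sound class and a shared first-person gunfire concurrency set.
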SoundCue /Game/Sounds/Fort_Weapons/Launchers/RPG/RPG_LT_Fire_1P_Cue
root=Modulator_0
nodes [
 {id=Modulator_0, children [Random_0], properties {volumeMax=1, volumeMin=1}}
 {id=Random_0, children [WavePlayer_0,WavePlayer_1,WavePlayer_3], properties {weights [1,1,1]}}
 {id=WavePlayer_0, properties {soundWaveAssetPtr=&/RPG_Fire_Close_03}}
 {id=WavePlayer_1, properties {soundWaveAssetPtr=&/RPG_Fire_Close_01}}
 {id=WavePlayer_3, properties {soundWaveAssetPtr=&/RPG_Fire_Close_02}}
]
properties {
 attenuationSettings=&Fort_Weapon_Projectile_FirstPerson_Athena_Att
 bDebug=true
 maxDistance=2097152
 soundClassObject=&Fort_WeaponRanged_Fire_1P_PSM
 volumeMultiplier=1.5
 concurrencyOverrides {bLimitToOwner=true, maxCount=3, resolutionRule=StopOldest}
 concurrencySet [&Weapon_Gunfire_FirstPerson_Conc]
}
