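# Sound cue for the end stage of the RPG launcher reload, by name the LT variant, classed as weapon foley.
# A single wave player for the RPG reload-end sound feeds a modulator that slightly randomizes volume and keeps the wave's pitch fixed.
# The cue lowers overall pitch, boosts volume, uses the shared weapon reload attenuation and overrides concurrency, with its debug flag left on.
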SoundCue /Game/Sounds/Fort_Weapons/Launchers/RPG/RPG_LT_Reload_End_Cue
root=Modulator_0
nodes [
 {id=Modulator_0, children [WavePlayer_2], properties {pitchMax=1, pitchMin=1, volumeMax=0.9, volumeMin=0.8}}
 {id=WavePlayer_2, properties {soundWaveAssetPtr=&/RPG_Reload_End}}
]
properties {
 attenuationSettings=&Fort_Weapon_Reload_Att
 bDebug=true
 bOverrideConcurrency=true
 maxDistance=2250
 pitchMultiplier=0.9
 soundClassObject=&Fort_WeaponFoley
 volumeMultiplier=1.1892
}
