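# Sound cue for the insert step of the RPG launcher reload foley in Fortnite's weapon audio.
# A single wave player feeds a modulator that adds slight random volume variation while pitch stays fixed.
# The cue lowers overall pitch and volume, routes to the weapon foley sound class, and uses shared reload attenuation with a concurrency override.
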SoundCue /Game/Sounds/Fort_Weapons/Launchers/RPG/RPG_LT_Reload_Insert_Cue
root=Modulator_0
nodes [
 {id=Modulator_0, children [WavePlayer_3], properties {pitchMax=1, pitchMin=1, volumeMax=1.3, volumeMin=1.2}}
 {id=WavePlayer_3, properties {soundWaveAssetPtr=&/RPG_Reload_Insert_v3}}
]
properties {
 attenuationSettings=&Fort_Weapon_Reload_Att
 bDebug=true
 bOverrideConcurrency=true
 maxDistance=2250
 pitchMultiplier=0.9
 soundClassObject=&Fort_WeaponFoley
 volumeMultiplier=0.852
}
