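# Sound cue for the start of the RPG launcher reload, classed as weapon foley.
# A single wave player feeds the RPG_Reload_Start_v5 wave into a modulator that slightly randomizes volume above unity and keeps pitch fixed.
# The cue lowers overall volume with its multiplier, uses the shared weapon reload attenuation, and overrides concurrency.
# The debug flag is left enabled.
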SoundCue /Game/Sounds/Fort_Weapons/Launchers/RPG/RPG_LT_Reload_Start_Cue
root=Modulator_0
nodes [
 {id=Modulator_0, children [WavePlayer_4], properties {pitchMax=1, pitchMin=1, volumeMax=1.3, volumeMin=1.2}}
 {id=WavePlayer_4, properties {soundWaveAssetPtr=&/RPG_Reload_Start_v5}}
]
properties {
 attenuationSettings=&Fort_Weapon_Reload_Att
 bDebug=true
 bOverrideConcurrency=true
 maxDistance=2250
 soundClassObject=&Fort_WeaponFoley
 volumeMultiplier=0.79
}
